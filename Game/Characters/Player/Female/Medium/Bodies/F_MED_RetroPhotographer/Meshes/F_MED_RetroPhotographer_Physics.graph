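# Physics asset for the medium female RetroPhotographer player body, built from capsules on the body, jacket fronts and camera strap rig.
# Pelvis, spine, arm and thigh capsules form the body, with extra capsules on cloned spine bones.
# Two-bone jacket chains hang from spine_04; a three-bone strap chain hangs from spine_05 via a base bone, branching into paired antenna chains.
# Accessory chains use tight angular limits, while clone, strap-base and forearm joints swing and twist freely.
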PhysicsAsset /Game/Characters/Player/Female/Medium/Bodies/F_MED_RetroPhotographer/Meshes/F_MED_RetroPhotographer_Physics
bodies {
 L_front_jacket_01 [{
   shapeType=Capsule
   length=3.2518
   radius=3.6053
   shapeName=L_front_jacket_01_capsule
   center=<2.5247 0.3216 2.3964>
   rotation=<rot -0.2475 -174.5133 -89.8771>
  }]
 L_front_jacket_02 [{
   shapeType=Capsule
   length=2.1835
   radius=3.6267
   shapeName=L_front_jacket_02_capsule
   center=<2.2618 0.0164 2.5573>
   rotation=<rot -4.219 6.3319 -90.8878>
  }]
 L_photoantennae_01 [{
   shapeType=Capsule
   length=0.8676
   radius=1.1076
   shapeName=L_photoantennae_01_capsule
   center=<1.7102 -0.0949 0.0071>
   rotation=<rot 4.6428 -90.0001 90.3959>
  }]
 L_photoantennae_02 [{
   shapeType=Capsule
   length=0.9235
   radius=0.9556
   shapeName=L_photoantennae_02_capsule
   center=<1.5589 -0.1299 0.0149>
   rotation=<rot 0 -90.7522 -88.3608>
  }]
 L_photoantennae_03 [{
   shapeType=Capsule
   length=1.8394
   radius=1.132
   shapeName=L_photoantennae_03_capsule
   center=<-0.0909 -0.1635 -1.8175>
   rotation=<rot 7.2366 -90 41.5875>
  }]
 R_front_jacket_01 [{
   shapeType=Capsule
   length=0.8
   radius=4.2891
   shapeName=R_front_jacket_01_capsule
   center=<3.3895 -0.1196 3.0381>
   rotation=<rot 4.9572 -0.4512 -89.6627>
  }]
 R_front_jacket_02 [{
   shapeType=Capsule
   length=4.2698
   radius=3.4826
   shapeName=R_front_jacket_02_capsule
   center=<3.0391 1.2629 2.3411>
   rotation=<rot 0 -2.1579 -87.9535>
  }]
 R_photoantennae_01 [{
   shapeType=Capsule
   length=0.8067
   radius=0.962
   shapeName=R_photoantennae_01_capsule
   center=<1.4034 -0.0986 -0.0987>
   rotation=<rot 11.18 -90.0002 -92.6422>
  }]
 R_photoantennae_02 [{
   shapeType=Capsule
   length=0.9445
   radius=1.0537
   shapeName=R_photoantennae_02_capsule
   center=<1.7381 -0.0793 0.0091>
   rotation=<rot 0 89.4234 -86.66>
  }]
 R_photoantennae_03 [{
   shapeType=Capsule
   length=1.7324
   radius=1.1306
   shapeName=R_photoantennae_03_capsule
   center=<1.877 -0.2431 0.5796>
   rotation=<rot 0 87.4274 -75.0538>
  }]
 lowerarm_l [{shapeType=Capsule, length=15.5, radius=4.505, shapeName=lowerarm_l_capsule, center=<12.4005 0 0>, rotation=<rot -0.0001 -89.9999 90>}]
 lowerarm_r [{shapeType=Capsule, length=15.5, radius=4.505, shapeName=lowerarm_l_capsule, center=<-12.4005 0 0>, rotation=<rot -0.0001 -89.9999 -90>}]
 pelvis [{shapeType=Capsule, length=12.7417, radius=9.9439, center=<-2.5532 -2.3826 0>}]
 photo_strap_01 [{shapeType=Capsule, length=11.505, radius=3.505, shapeName=photo_strap_01_capsule, center=<2.4683 2.1359 0>, rotation=<rot 0 -90 0>}]
 photo_strap_01_BASE [{shapeType=Capsule, length=0.505, radius=0.505, shapeName=photo_strap_01_BASE_capsule, rotation=<rot 0 -90 0>}]
 photo_strap_02 [{
   shapeType=Capsule
   length=9.4349
   radius=4.9367
   shapeName=photo_strap_02_capsule
   center=<5.6014 2.1579 0.1737>
   rotation=<rot -2.0521 -89.9999 0.1841>
  }]
 photo_strap_03 [{
   shapeType=Capsule
   length=3.107
   radius=6.709
   shapeName=photo_strap_03_capsule
   center=<12.4399 1.4863 0.7608>
   rotation=<rot -3.3571 -90.5539 177.617>
  }]
 spine_02 [
  {shapeType=Capsule, length=13.9934, radius=10.0917, center=<0 -1.1181 0>}
  {
   shapeType=Capsule
   length=2.636
   radius=4.8179
   shapeName=spine_02_capsule
   center=<-0.0661 4.9731 8.6939>
   rotation=<rot 0 -0.0001 -54.7501>
  }
 ]
 spine_03_clone_01 [{shapeType=Capsule, length=1.5, radius=9.6, shapeName=spine_03_clone_01_capsule, rotation=<rot 0 -90 0>}]
 spine_04 [{shapeType=Capsule, length=12.7128, radius=10.714, center=<0 0.0043 0>}]
 spine_04_clone_01 [{shapeType=Capsule, length=3.5, radius=9.6, shapeName=spine_04_clone_01_capsule, rotation=<rot 0 -90 0>}]
 spine_05 [{shapeType=Capsule, length=14.4246, radius=8.9706, center=<-0.348 -0.0859 0>}]
 thigh_l [{shapeType=Capsule, length=23.2334, radius=9.5774, center=<-24.5058 -0.8727 0.0293>, rotation=<rot 0.0001 -88.0954 -92.3192>}]
 thigh_r [{shapeType=Capsule, length=23.2334, radius=9.5774, center=<24.5058 0.8727 -0.0293>, rotation=<rot 0.0001 -88.0954 87.6808>}]
 upperarm_l [{shapeType=Capsule, length=14.2625, radius=5.5388, center=<13.0817 -0.4846 -0.9178>, rotation=<rot 0.3007 89.6677 -88.3035>}]
 upperarm_r [{shapeType=Capsule, length=14.2625, radius=5.5388, center=<-13.0817 0.4846 0.9178>, rotation=<rot 0.3007 89.6677 91.6965>}]
}
constraints [
 {bone1Name=spine_02, bone2Name=pelvis, swing1LimitDegrees=45, swing2LimitDegrees=45, twistLimitDegrees=45}
 {bone1Name=spine_04, bone2Name=spine_02, swing1LimitDegrees=45, swing2LimitDegrees=45, twistLimitDegrees=45}
 {bone1Name=spine_05, bone2Name=spine_04, swing1LimitDegrees=45, swing2LimitDegrees=45, twistLimitDegrees=45}
 {bone1Name=upperarm_l, bone2Name=spine_05, swing1LimitDegrees=45, swing2LimitDegrees=45, twistLimitDegrees=45}
 {bone1Name=thigh_l, bone2Name=pelvis, swing1LimitDegrees=45, swing2LimitDegrees=45, twistLimitDegrees=45}
 {bone1Name=thigh_r, bone2Name=pelvis, swing1LimitDegrees=45, swing2LimitDegrees=45, twistLimitDegrees=45}
 {bone1Name=upperarm_r, bone2Name=spine_05, swing1LimitDegrees=45, swing2LimitDegrees=45, twistLimitDegrees=45}
 {
  bone1Name=spine_03_clone_01
  bone2Name=spine_02
  swing1LimitDegrees=45
  swing1Motion=Free
  swing2LimitDegrees=45
  swing2Motion=Free
  twistLimitDegrees=45
  twistMotion=Free
 }
 {
  bone1Name=spine_04_clone_01
  bone2Name=spine_04
  swing1LimitDegrees=45
  swing1Motion=Free
  swing2LimitDegrees=45
  swing2Motion=Free
  twistLimitDegrees=45
  twistMotion=Free
 }
 {bone1Name=L_front_jacket_01, bone2Name=spine_04, swing1LimitDegrees=30, swing2LimitDegrees=30, twistLimitDegrees=10}
 {bone1Name=L_front_jacket_02, bone2Name=L_front_jacket_01, swing1LimitDegrees=10, swing2LimitDegrees=10, twistLimitDegrees=5}
 {bone1Name=R_front_jacket_02, bone2Name=R_front_jacket_01, swing1LimitDegrees=20, swing2LimitDegrees=20, twistLimitDegrees=5}
 {bone1Name=R_front_jacket_01, bone2Name=spine_04, swing1LimitDegrees=30, swing2LimitDegrees=30, twistLimitDegrees=5}
 {bone1Name=photo_strap_01, bone2Name=photo_strap_01_BASE, swing1LimitDegrees=15, swing2LimitDegrees=10, twistLimitDegrees=5}
 {bone1Name=photo_strap_02, bone2Name=photo_strap_01, swing1LimitDegrees=10, swing2LimitDegrees=10, twistLimitDegrees=5}
 {bone1Name=photo_strap_03, bone2Name=photo_strap_02, swing1LimitDegrees=10, swing2LimitDegrees=10, twistLimitDegrees=5}
 {bone1Name=L_photoantennae_01, bone2Name=photo_strap_03, swing1LimitDegrees=10, swing2LimitDegrees=10, twistLimitDegrees=5}
 {bone1Name=L_photoantennae_02, bone2Name=L_photoantennae_01, swing1LimitDegrees=10, swing2LimitDegrees=10, twistLimitDegrees=5}
 {bone1Name=L_photoantennae_03, bone2Name=L_photoantennae_02, swing1LimitDegrees=10, swing2LimitDegrees=10, twistLimitDegrees=5}
 {bone1Name=R_photoantennae_01, bone2Name=photo_strap_03, swing1LimitDegrees=10, swing2LimitDegrees=10, twistLimitDegrees=5}
 {bone1Name=R_photoantennae_02, bone2Name=R_photoantennae_01, swing1LimitDegrees=10, swing2LimitDegrees=10, twistLimitDegrees=5}
 {bone1Name=R_photoantennae_03, bone2Name=R_photoantennae_02, swing1LimitDegrees=10, swing2LimitDegrees=10, twistLimitDegrees=5}
 {
  bone1Name=photo_strap_01_BASE
  bone2Name=spine_05
  swing1LimitDegrees=45
  swing1Motion=Free
  swing2LimitDegrees=45
  swing2Motion=Free
  twistLimitDegrees=45
  twistMotion=Free
 }
 {
  bone1Name=lowerarm_l
  bone2Name=upperarm_l
  swing1LimitDegrees=45
  swing1Motion=Free
  swing2LimitDegrees=45
  swing2Motion=Free
  twistLimitDegrees=45
  twistMotion=Free
 }
 {
  bone1Name=lowerarm_r
  bone2Name=upperarm_r
  swing1LimitDegrees=45
  swing1Motion=Free
  swing2LimitDegrees=45
  swing2Motion=Free
  twistLimitDegrees=45
  twistMotion=Free
 }
]
