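# Post-process style AnimBP for the female RetroPhotographer body part, adding secondary jacket, pendant and cloth motion over the parent mesh's pose.
# The copied pose feeds a deform Control Rig driven by parachute, crouch and DBNO states, then one of two RigidBody sims chosen by a front-end idle flag.
# A jacket bone offset and arm-pendant AnimDynamics finish the pose.
# The EventGraph drives material-based WPO cloth, adding wind and turbulence while skydiving.
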AnimBlueprint /Game/Characters/Player/Female/Medium/Bodies/F_MED_RetroPhotographer/Meshes/F_MED_RetroPhotographer_AnimBP
target_skeleton=&/F_MED_RetroPhotographer_Skeleton
graphs [
 AnimGraph {nodes [
   n1 = RigidBody!(
    ComponentPose=Local To Component(LocalPose=Use cached pose 'Rigid_body'())
    Alpha
    OverrideWorldGravity:Vector=Get Gravity Override()
    ExternalForce
   ) -> (Pose) {
    bClampLinearTranslationLimitToRefPose=true
    bOverrideWorldGravity=true
    overridePhysicsAsset=&/F_MED_RetroPhotographer_Physics
    simulationSpace=BaseBoneSpace
    alphaBoolBlend {blendOption=Linear}
    baseBoneRef {boneName=spine_03}
    componentAppliedLinearAccClamp=<2000 2000 2000>
    componentLinearAccScale=<0.3 0.3 0.3>
    componentLinearVelScale=<0.3 0.3 0.3>
    overrideWorldGravity {z=-980}
   }
   n2 = Component To Local(ComponentPose=@n1)
   n3 = RigidBody(ComponentPose=Local To Component(LocalPose=Use cached pose 'Rigid_body'()), OverrideWorldGravity=Get Gravity Override()) {
    bClampLinearTranslationLimitToRefPose=true
    bOverrideWorldGravity=true
    overridePhysicsAsset=&F_MED_RetroPhotographer_Physics
    simulationSpace=BaseBoneSpace
    alphaBoolBlend {blendOption=Linear}
    baseBoneRef {boneName=spine_05}
    componentAppliedLinearAccClamp=<2000 2000 2000>
    componentLinearAccScale=<0.3 0.3 0.3>
    componentLinearVelScale=<0.3 0.3 0.3>
    overrideWorldGravity {z=-980}
   }
   n4 = Component To Local(ComponentPose=@n3)
   n5 = Get FrontEnd Idle Bypass() -> (FrontEnd Idle Bypass:bool)
   n6 = "Transform (Modify) Bone - Bone: R_front_jacket_01_clone"(
    ComponentPose=Local To Component(LocalPose=Blend Poses by bool(BlendPose_0=@n4, BlendPose_1=@n2, bActiveValue=@n5) {blendType=Linear})
    Alpha=1
    Translation="0,-1.5,2.5"
    Rotation="0,0,0"
    Scale="1,1,1"
   ) {
    rotationMode=BMM_Additive
    rotationSpace=BCS_BoneSpace
    translationMode=BMM_Additive
    translationSpace=BCS_BoneSpace
    alphaBoolBlend {blendInTime=0.1, blendOption=Linear, blendOutTime=0.2}
    alphaScaleBiasClamp {bInterpResult=true, interpSpeedDecreasing=20, interpSpeedIncreasing=20}
    boneToModify {boneName=R_front_jacket_01_clone}
    translation {y=-1.5, z=2.5}
   }
   n7 = Anim Dynamics - Bone: zipper(ComponentPose=@n6, Alpha=0, ExternalForce="0,0,0") {
    angularDampingOverride=0.9999
    angularSpringConstant=50
    bAngularSpring=true
    bOverrideAngularDamping=true
    linearDampingOverride=0.999
    linearSpringConstant=60
    simulationSpace=RootRelative
    alphaBoolBlend {blendOption=Linear}
    alphaScaleBiasClamp {scale=0.2}
    boundBone {boneName=zipper}
    componentAppliedLinearAccClamp=<6000 6000 6000>
    componentLinearAccScale=<0.4 0.4 0.4>
    componentLinearVelScale=<0.4 0.4 0.4>
    gravityOverride {z=-400}
    physicsBodyDefinitions [{
      boundBone {boneName=zipper}
      boxExtents=<5 2 2>
      constraintSetup {angularLimitsMax=<0 0 30>, angularLimitsMin=<0 0 -30>, angularTarget=<1 0 0>}
      localJointOffset=<4 0 0>
     }]
   }
   n8 = Anim Dynamics - Bone: L_arm_pendant(ComponentPose=@n7, Alpha=0, ExternalForce="0,0,0") {
    angularDampingOverride=0.1
    bAngularSpring=true
    bOverrideAngularDamping=true
    linearDampingOverride=0.999
    linearSpringConstant=60
    simulationSpace=World
    alphaBoolBlend {blendOption=Linear}
    boundBone {boneName=L_arm_pendant}
    componentAppliedLinearAccClamp=<6000 6000 6000>
    componentLinearAccScale=<0.4 0.4 0.4>
    componentLinearVelScale=<0.4 0.4 0.4>
    gravityOverride {z=-400}
    physicsBodyDefinitions [{
      boundBone {boneName=L_arm_pendant}
      boxExtents=<5 2 2>
      constraintSetup {angularLimitsMax=<30 30 30>, angularLimitsMin=<-30 -30 -30>, angularTarget=<1 0 0>}
      localJointOffset=<2 0 0>
     }]
    relativeSpaceBone {boneName=spine_05}
   }
   n9 = Anim Dynamics - Bone: L_arm_pendant(ComponentPose=@n8, Alpha=1, ExternalForce="0,0,0") {
    angularDampingOverride=0.9999
    angularSpringConstant=0.5
    bGravityOverrideInSimSpace=true
    bOverrideAngularDamping=true
    bOverrideLinearDamping=true
    linearDampingOverride=0.9999
    simulationSpace=RootRelative
    alphaBoolBlend {blendOption=Linear}
    boundBone {boneName=L_arm_pendant}
    componentAppliedLinearAccClamp=<10000 10000 10000>
    componentLinearAccScale=<0.15 0.15 0.15>
    componentLinearVelScale=<0.15 0.15 0.15>
    gravityOverride=<1 1 1>
    physicsBodyDefinitions [{
      sphereCollisionRadius=0
      boundBone {boneName=L_arm_pendant}
      boxExtents=<6 6 2>
      constraintSetup {angularLimitsMax=<70 70 100>, angularLimitsMin=<-70 -70 -70>, angularTarget=<0 0 0>}
      localJointOffset=<2.5 0 0>
     }]
    relativeSpaceBone {boneName=root}
   }
   n10 = Anim Dynamics - Bone: R_arm_pendant(ComponentPose=@n9, Alpha=1, ExternalForce="0,0,0") {
    angularDampingOverride=0.9999
    angularSpringConstant=0.5
    bGravityOverrideInSimSpace=true
    bOverrideAngularDamping=true
    bOverrideLinearDamping=true
    linearDampingOverride=0.9999
    simulationSpace=RootRelative
    alphaBoolBlend {blendOption=Linear}
    boundBone {boneName=R_arm_pendant}
    componentAppliedLinearAccClamp=<10000 10000 10000>
    componentLinearAccScale=<0.15 0.15 0.15>
    componentLinearVelScale=<0.15 0.15 0.15>
    gravityOverride=<1 1 1>
    physicsBodyDefinitions [{
      sphereCollisionRadius=0
      boundBone {boneName=R_arm_pendant}
      boxExtents=<6 6 2>
      constraintSetup {angularLimitsMax=<30 40 30>, angularLimitsMin=<-30 -70 -100>, angularTarget=<0 0 0>}
      localJointOffset=<2.5 0 0>
     }]
    relativeSpaceBone {boneName=root}
   }
   Output Pose(Result=Component To Local(ComponentPose=@n10)) {layerGroup=None}
   n11 = Control Rig!(
    Source=Copy Pose From Mesh() {bUseAttachedParent=true}
    Alpha
    parachute:double=Get parachute_float() -> (parachute_float:double)
    crouching:double=Get Crouching_float() -> (Crouching_float:double)
    DBNO_float:double=Get DBNO_float() -> (DBNO_float:double)
   ) -> (Pose) {
    alphaBoolBlend {blendOption=Linear}
    controlRigAssetReference {blueprintRigClass=&/F_MED_RetroPhotographer_DeformRig.F_MED_RetroPhotographer_DeformRig_C}
   }
   Save cached pose 'Rigid_body'(Pose=@n11) {cacheName=Rigid_body}
  ]}
 EventGraph {nodes [
   n1 = Event BlueprintUpdateAnimation()
   > n2 = Sequence!(execute) -> (then_0, then_1, then_2:Exec, then_3:Exec, then_4:Exec)
   > n3 = Gravity Override State Function DEPRECATED(CharacterPartAnimInstance=Self-Reference(), GameTimeInSeconds=0, GravityOverrideParams=Get Gravity Override Params())
   > Set Gravity Override(Gravity Override=@n3)
   > Branch(Condition=Get bIsInFrontEnd())
   > Set FrontEnd Idle Bypass(FrontEnd Idle Bypass:bool=NOT Boolean(A=Get bIsPlayingEmote())) -> (Output_Get:bool)
   Set parachute_float(execute=@n2.then_1, parachute_float:double="To Float (Boolean)"(InBool=Get bIsParachuteOpened())) -> (Output_Get:double)
   > Set Crouching_float(Crouching_float:double="To Float (Boolean)"(InBool=Get bIsCrouching())) -> (Output_Get:double)
   > Set DBNO_float(DBNO_float:double="To Float (Boolean)"(InBool=Get bIsDBNO())) -> (Output_Get:double)
   n4 = Branch(execute=@n2.then_2, Condition=Get bIsSkyDiving())
   > Set Velocity Interp Speed(Velocity Interp Speed:double=15) -> (Output_Get:double)
   > Set TurbulenceNoiseTileDivisor(TurbulenceNoiseTileDivisor:float=15) -> (Output_Get:float)
   > Set wind_strength(wind_strength:float=0.9) -> (Output_Get:float)
   Set Velocity Interp Speed(execute=@n4.else, Velocity Interp Speed=0)
   > Set TurbulenceNoiseTileDivisor(TurbulenceNoiseTileDivisor=0)
   > Set wind_strength(wind_strength=0)
   Set WPO Globals Function(
    execute=@n2.then_4
    Material=Get Mat Instance Dynamic()
    Overlay Material=Get Overlay Mat Instance Dynamic()
    Skeletal Mesh=Get SkelMesh()
    "Local Noise Transform  Location"="Get Local Noise Transform  Location"() -> ("Local Noise Transform  Location":Vector)
    Local Noise Offset=Get Local Noise Offset()
   )
   > Set WPO Regional Function!(
    execute
    Material:Material Instance Dynamic Object Reference=Get Mat Instance Dynamic()
    Overlay Material:Material Instance Dynamic Object Reference=Get Overlay Mat Instance Dynamic()
    Skeletal Mesh:Skeletal Mesh Component Object Reference=Get SkelMesh()
    Velocity Transform:Name=calf_r
    Velocity Interp Speed:double=Get Velocity Interp Speed()
    Collision Transform:Name=calf_r
    Collision Transform Axis:EAxis Enum
    Invert Collision Transform Axis:bool=false
    Delta Time X:double=@n1.DeltaTimeX
    Previous Velocity Transform Location:"Vector (by ref)"=Get R1 Velocity Transform Location()
    Previous Velocity Transform Velocity:"Vector (by ref)"=Get R1 Velocity Transform Velocity()
    Region:int=1
   ) -> (then)
   n5 = Get Overlay Mat Instance Dynamic()
   n6 = Get Mat Instance Dynamic()
   > SetScalarParameterValue(self=[@n5, @n6], ParameterName=ClothSim_WindStrength, Value=Get wind_strength() -> (wind_strength:float))
   > SetScalarParameterValue(
    self=[@n5, @n6]
    ParameterName=ClothSim_TurbulenceNoiseTileDivisor
    Value=Get TurbulenceNoiseTileDivisor() -> (TurbulenceNoiseTileDivisor:float)
   )
   Event OnCharacterCustomizationLoaded()
   > n7 = Set SkelMesh(SkelMesh=GetPartSkeletalMeshComponent())
   > n8 = Get Dynamic Material Instance(SkeletalMeshComponent=@n7)
   > Set Mat Instance Dynamic(Mat Instance Dynamic=@n8)
   > n9 = GetOrCreateDynamicOverlayMaterial(MeshComponent=@n7, MaterialName=None)
   > Set Overlay Mat Instance Dynamic(Overlay Mat Instance Dynamic=@n9)
   n10 = Branch(Condition=Get bIsSkydiveFloating())
   > n11 = Set WPO Pho Wind Intensity(WPO Pho Wind Intensity=1)
   SetScalarParameterValue(
    execute=[@n11.then, @n12.then]
    self=[Get Mat Instance Dynamic(), Get Overlay Mat Instance Dynamic()]
    ParameterName=ClothSim_WindStrength
    Value=Get WPO Pho Wind Intensity()
   )
   n12 = Set WPO Pho Wind Intensity(execute=@n10.else, WPO Pho Wind Intensity=0)
   n13 = Branch(Condition=NOT Boolean(A=Get bIsSkyDiving()))
   > n14 = Set WPO Pho Wind Intensity(WPO Pho Wind Intensity=0)
   n15 = Get Mat Instance Dynamic()
   n16 = Get Overlay Mat Instance Dynamic()
   n17 = Get WPO Pho Wind Intensity()
   SetScalarParameterValue(execute=[@n18.then, @n14.then], self=[@n15, @n16], ParameterName=ClothSim_R1_DragStrength, Value=@n17)
   > SetScalarParameterValue(self=[@n15, @n16], ParameterName=ClothSim_R2_DragStrength, Value=@n17)
   n18 = Set WPO Pho Wind Intensity(execute=@n13.else, WPO Pho Wind Intensity=1)
   OR Boolean!(A=Get bIsSkydiveFloating(), B=Get bIsSkydiveDiveMode(), C:bool=Get bIsParachuteOpened()) -> (ReturnValue)
   AND Boolean(A=Get bIsJumping(), B=NOT Boolean(A=false))
  ]}
]
variables [
 Crouching_float {type=number}
 DBNO_float {type=number}
 FrontEnd Idle Bypass {type=boolean}
 Gravity Override {type=Vector}
 Gravity Override Params {type=GravityOverrideParamsStruct}
 Local Noise Offset {type=Vector}
 "Local Noise Transform  Location" {type=Vector}
 Mat Instance Dynamic {type=MaterialInstanceDynamic}
 Overlay Mat Instance Dynamic {type=MaterialInstanceDynamic}
 parachute_float {type=number}
 R1 Velocity Transform Location {type=Vector}
 R1 Velocity Transform Velocity {type=Vector}
 R2 Velocity Transform Location {type=Vector}
 R2 Velocity Transform Velocity {type=Vector}
 SkelMesh {type=SkeletalMeshComponent}
 TurbulenceNoiseTileDivisor {type=number}
 Velocity Interp Speed {type=number}
 wind_strength {type=number}
 WPO Pho Wind Intensity {type=number}
]
